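# Masked, two-sided material that fakes sloshing liquid inside a container, for IronLog weapon tank effects.
# Niagara dynamic parameters, or static overrides, drive water plane, fill and motion; a wavy water-line gradient clips opacity.
# Virtual-plane coordinates and transformed normals fake the top surface, with meniscus and switchable foam.
# Panned bubble layers, sky-light environment sampling for fake refraction, gradient color and fresnel glow shade the liquid, with camera-offset WPO.
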Material /WildEstateWeapons/Gameplay/IronLog/FX/Tanks/M_Liquid_Pendulum_IronLog
graph {
 properties {
  bUsedWithMeshDeformer=true
  bUsedWithNiagaraMeshParticles=true
  bUsedWithSkeletalMesh=true
  blendMode=BLEND_Masked
  pixelDepthOffsetMode=PDOM_Legacy
  twoSided=true
  usedShadingModels=Default Lit
 }
 nodes [
  n1 = PreSkinnedPosition()
  n2 = ComponentMask(None=Divide(A=VertexInterpolator(VS=@n1).PS, B=NamedRerouteUsage())) {b=true}
  Abs(None=Tangent!(None=Multiply(B=@n2) {constA=0.125}))
  n3 = Multiply(A=NamedRerouteUsage(), B=NamedRerouteUsage())
  Add(A=@n3, B=NamedRerouteUsage())
  Constant2Vector() {g=1, r=1}
  Constant3Vector() {constant=<rgba 0 0 -1 0>}
  Constant() {r=1}
  Constant() {r=0.1592}
  DynamicParameter() {parameterIndex=3}
  n4 = NamedRerouteUsage()
  n5 = Multiply(A=Multiply(A=NamedRerouteUsage(), B=Constant2Vector() {g=0.3, r=0.4}), B=LinearInterpolate(Alpha=@n4) {constA=1, constB=0.5})
  n6 = NamedRerouteUsage()
  n7 = NamedRerouteUsage()
  n8 = LinearInterpolate(Alpha=@n4) {constA=2}
  n9 = Constant3Vector() {desc=Normal Flat, constant=<rgba 0 0 1 0>}
  n10 = StaticSwitchParameter(True=@n9, False=@n9) {parameterName=Enable Bubbles}
  LinearInterpolate(
   A=@n10
   B=TextureSampleParameter2D(UVs=Panner(Coordinate=@n5, Time=@n6) {speedY=-1.75}) {parameterName=Normal_Top, samplerType=SAMPLERTYPE_Normal, texture=/Game/Global/Textures/Default/Noise/T_Bubble_B_N}.RGB
   Alpha=Clamp(None=Power(Base=@n7, Exp=@n8) {constExponent=1.5}) {maxDefault=0.5}
  ) {constAlpha=1}
  LinearInterpolate(Alpha=NamedRerouteUsage()) {constA=0.05, constB=0.1}
  MaterialFunctionCall() {materialFunction=&Pi}
  n11 = NamedRerouteUsage()
  n12 = Abs(None=DotProduct(A=Normalize(VectorInput=@n11), B=CameraVectorWS()))
  n13 = NamedRerouteUsage()
  n14 = Subtract(A=OneMinus(None=Multiply(A=@n12, B=@n13)), B=@n13)
  n15 = Multiply(A=@n13) {constB=0}
  n16 = Saturate(None=StaticSwitchParameter(True=Subtract(A=@n14, B=@n15), False=@n14) {parameterName=Enable Bubbles})
  n17 = NamedRerouteUsage()
  n18 = NamedRerouteUsage()
  n19 = StaticSwitchParameter(True=LinearInterpolate(A=@n16, Alpha=Multiply(A=@n17, B=@n18) {constA=0.5}), False=@n16) {parameterName=Enable Foam}
  n20 = NamedRerouteUsage()
  Max(A=@n19, B=LinearInterpolate(Alpha=Power(Base=@n20)) {constB=0.8})
  Multiply(A=NamedRerouteUsage()) {constB=0.1}
  Multiply()
  n21 = NamedRerouteUsage()
  n22 = Power(Base=OneMinus(None=NamedRerouteUsage())) {constExponent=3}
  Multiply(
   A=Multiply(A=Multiply(A=@n21, B=@n22), B=ScalarParameter() {group=Bubbles, parameterName=Bubbles Glow}) {constB=0.05}
   B=Saturate(None=Add() {constB=0.25})
  )
  n23 = NamedRerouteUsage()
  n24 = Normalize(VectorInput=@n23)
  n25 = DotProduct(A=@n24, B=Constant3Vector() {constant=<rgba 0 0 1 0>})
  Multiply(A=Add(A=@n25) {constB=0}) {constB=-0.15}
  n26 = LinearInterpolate(Alpha=NamedRerouteUsage()) {constA=0.5, constAlpha=1}
  Multiply(A=@n26, B=LinearInterpolate(Alpha=NamedRerouteUsage()) {constA=1, constB=0.5})
  Multiply(A=Power(Base=NamedRerouteUsage()), B=Constant2Vector() {g=1.75})
  n27 = Multiply(A=Multiply(A=NamedRerouteUsage(), B=ScalarParameter() {defaultValue=1, parameterName=Water Surface Normal Tile}), B=@n26)
  n28 = Divide(A=NamedRerouteUsage(), B=@n27)
  n29 = Multiply(A=MaterialFunctionCall(Vector 3=@n28, Vector 2=@n28) {materialFunction=/Engine/Functions/Engine_MaterialFunctions02/Utility/VectorLength}.V3 Length) {constB=3}
  Multiply(A=Saturate(None=OneMinus(None=@n29)), B=NamedRerouteUsage())
  n30 = DynamicParameter() {parameterIndex=2, defaultValue=<rgba 0 0 0 0>, paramNames [Speed,Turbulence,Param3,Param4]}
  NamedRerouteDeclaration(None=@n30.Speed) {name=Speed, nodeColor=<rgba 0 0.7891 1 1>}
  NamedRerouteDeclaration(None=Saturate(None=TwoSidedSign())) {name=Backface, nodeColor=<rgba 1 0 0.7031 1>}
  n31 = DynamicParameter() {parameterIndex=1, defaultValue=<rgba 1 0 0 0>, paramNames [Tangent X, Tangent Y, Tangent Z, Time]}
  NamedRerouteDeclaration(None=MaterialFunctionCall(X=@n31.Tangent X, Y=@n31.Tangent X, Z=@n31.Tangent X) {materialFunction=&MakeFloat3}.Result) {name=WaterPlaneTangent, nodeColor=<rgba 0.8984 1 0 1>}
  n32 = Ceil(None=NamedRerouteUsage())
  n33 = NamedRerouteUsage()
  n34 = LinearInterpolate(Alpha=NamedRerouteUsage()) {constA=0.5, constB=-0.5}
  n35 = VectorParameter() {parameterName=Container Is Flat}
  n36 = NamedRerouteUsage()
  n37 = LinearInterpolate(
   A=Multiply(A=Tangent(None=@n34) {period=3}) {constB=0.5}
   B=@n34
   Alpha=Length(None=Multiply(A=@n35.RGB, B=@n36)) {is=class MaterialExpressionLength}
  )
  n38 = Subtract(A=Multiply(A=ComponentMask(None=@n33) {b=true}) {constA=-1, constB=-1}, B=@n37) {constB=0}
  n39 = Abs(None=@n37)
  n40 = OneMinus(None=Power(Base=Multiply(A=@n39) {constB=2}) {constExponent=3})
  n41 = NamedRerouteUsage()
  n42 = NamedRerouteUsage()
  n43 = Constant3Vector() {constant=<rgba 1 1 0 0>}
  n44 = Clamp(None=DotProduct(A=Normalize(VectorInput=@n41), B=Multiply(A=@n42, B=@n43))) {minDefault=-1}
  n45 = Power(Base=Abs(None=@n44), Exp=ScalarParameter() {defaultValue=2.5, group=Debug, parameterName=Bend Exponent}) {constExponent=3}
  n46 = NamedRerouteUsage()
  n47 = Power(Base=@n46) {constExponent=0.4}
  n48 = Power(Base=@n46) {constExponent=0.5}
  n49 = Add(A=@n44, B=Multiply(A=NamedRerouteUsage()) {constB=3}) {constB=0}
  n50 = Multiply(A=Sine(None=Multiply(A=@n49) {constB=-1.25}), B=LinearInterpolate(Alpha=NamedRerouteUsage()) {constA=1, constB=0.25})
  n51 = Add(
   A=Multiply(A=ConstantBiasScale(None=@n45) {bias=-0.666, scale=2}, B=Multiply(A=@n47) {constA=1, constB=0.2}) {constB=0.15}
   B=Multiply(A=Multiply(A=@n48, B=@n50) {constA=0.2}) {constB=0.075}
  ) {constB=0}
  n52 = Multiply(A=Multiply(A=NamedRerouteUsage(), B=Constant2Vector() {g=1.2, r=1.5}), B=LinearInterpolate(Alpha=NamedRerouteUsage()) {constA=1, constB=0.6})
  n53 = NamedRerouteUsage()
  n54 = NamedRerouteDeclaration(None=TextureSample(UVs=Panner(Coordinate=@n52, Time=@n53) {speedY=-3}) {samplerType=SAMPLERTYPE_Grayscale, texture=&LowResBlurredNoise}.R) {name=Water Breakup, nodeColor=<rgba 0 1 0.5313 1>}
  n55 = Max(A=Power(Base=NamedRerouteUsage()), B=Multiply(A=NamedRerouteUsage()) {constB=0.1})
  n56 = StaticSwitchParameter(True=Add(A=Multiply(A=@n40, B=@n51) {constA=1}, B=Multiply(A=@n54, B=@n55)) {constB=0}, False=Constant()) {defaultValue=true, parameterName=Is Niagara}
  n57 = Add(A=NamedRerouteDeclaration(None=Add(A=@n38, B=@n56) {constB=0}) {name=Water Line Gradient, nodeColor=<rgba 1 0 0.1172 1>}, B=@n3)
  NamedRerouteDeclaration(None=Multiply(A=Saturate(None=@n32), B=SmoothStep(Value=@n57) {constMax=0.001}) {constA=1}) {name=Opacity, nodeColor=<rgba 0.4297 1 0 1>}
  n58 = ScreenPosition()
  n59 = MaterialFunctionCall(Float2=ViewSize()) {materialFunction=&BreakOutFloat2Components}
  n60 = AppendVector(A=Constant() {r=1}, B=Divide(A=@n59.G, B=@n59.G))
  n61 = ConstantBiasScale(None=Multiply(A=ConstantBiasScale(None=@n58.ViewportUV) {bias=-0.5, scale=2}, B=@n60))
  n62 = Multiply(A=NamedRerouteUsage(), B=NamedRerouteUsage())
  n63 = ScalarParameter() {defaultValue=0.1, parameterName=Screen Space UV Wobble}
  n64 = LinearInterpolate(Alpha=NamedRerouteUsage()) {constA=0.1, constB=0.5}
  n65 = Subtract(A=Add(A=NamedRerouteUsage(), B=NamedRerouteUsage()), B=Multiply(A=Multiply(A=@n62), B=Max(A=@n63, B=@n64)))
  n66 = Multiply(
   A=@n60
   B=Multiply(
    A=Constant2Vector() {g=-0.5, r=0.5}
    B=MaterialFunctionCall(World Position=@n65) {materialFunction=&TransformToClipSpace}.Clip Space XY
   )
  )
  n67 = ScalarParameter() {defaultValue=1, parameterName=Screen Space Scale}
  n68 = Divide(
   A=Multiply(A=NamedRerouteUsage(), B=Multiply(A=@n67) {constB=16})
   B=ComponentMask(None=TransformPosition(None=@n65) {transformSourceType=TRANSFORMPOSSOURCE_World, transformType=TRANSFORMPOSSOURCE_View}) {b=true}
  ) {constA=100}
  NamedRerouteDeclaration(None=MaterialFunctionCall(UVs=Subtract(A=@n61, B=@n66), Texture Scale=Multiply(A=@n68)) {materialFunction=&ScaleUVsByCenter}.UVs) {name=ScreenSpaceUV, nodeColor=<rgba 1 0 0.8672 1>}
  n69 = DotProduct(A=Constant3Vector() {constant=<rgba 0 0 -1 0>}, B=NamedRerouteUsage())
  NamedRerouteDeclaration(None=Multiply(A=Add(A=@n69)) {constB=0.5}) {name=WaterUpsideDown, nodeColor=<rgba 0 0.1797 1 1>}
  NamedRerouteDeclaration(None=StaticSwitchParameter(True=ParticlePositionWS(), False=ObjectPositionWS()) {defaultValue=true, parameterName=Is Niagara}) {name=Position, nodeColor=<rgba 0 0.5078 1 1>}
  n70 = Subtract(A=Subtract(A=WorldPosition() {worldPositionShaderOffset=WPT_ExcludeAllShaderOffsets}.XYZ, B=NamedRerouteUsage()), B=NamedRerouteUsage())
  n71 = Multiply(A=NamedRerouteUsage(), B=Constant3Vector() {constant=<rgba 1 1 -1 0>})
  n72 = Multiply(A=NamedRerouteUsage(), B=Constant3Vector() {constant=<rgba -1 -1 1 0>})
  n73 = MaterialFunctionCall(VectorToTransform=@n70, BasisX=@n71, BasisY=CrossProduct(A=@n72, B=@n71), BasisZ=@n72) {materialFunction=&Transform3x3Matrix}
  NamedRerouteDeclaration(None=Divide(A=@n73.Transformed Vector, B=NamedRerouteUsage()) {constB=1}) {name=LiquidCoordinates, nodeColor=<rgba 0 1 0.1094 1>}
  n74 = NamedRerouteUsage()
  n75 = NamedRerouteUsage()
  n76 = Multiply(
   A=NamedRerouteUsage()
   B=LinearInterpolate(
    A=ScalarParameter() {group=Foam, parameterName=Foam Min}
    B=ScalarParameter() {defaultValue=0.4, group=Foam, parameterName=Foam Max}
    Alpha=Power(Base=@n75)
   ) {constA=10, constB=5}
  )
  n77 = Constant()
  n78 = NamedRerouteUsage()
  n79 = NamedRerouteUsage()
  n80 = Divide(A=ScalarParameter() {defaultValue=0.5, parameterName=Miniscus Thickness}, B=Length(None=Multiply(A=@n78, B=@n79)))
  n81 = Subtract(A=Subtract(A=@n74, B=StaticSwitchParameter(True=@n76, False=@n77) {parameterName=Enable Foam}) {constB=0}, B=@n80) {constB=0}
  n82 = Power(Base=Saturate(None=Divide(A=@n81, B=@n80) {constB=0.015})) {constExponent=1}
  n83 = NamedRerouteUsage()
  n84 = OneMinus(None=Saturate(None=Add(A=@n82, B=@n83) {constB=0}))
  n85 = Divide(A=@n74, B=@n80) {constB=0.02}
  n86 = Multiply(A=OneMinus(None=Saturate(None=@n85)))
  NamedRerouteDeclaration(None=LinearInterpolate(Alpha=Subtract(A=@n84, B=@n86) {constA=0}) {constB=0.25}) {name=Miniscus, nodeColor=<rgba 0 1 0.5313 1>}
  n87 = NamedRerouteUsage()
  n88 = NamedRerouteUsage()
  n89 = MaterialFunctionCall(Emissive=Multiply(A=Multiply(A=@n87, B=@n88) {constA=1}, B=ScalarParameter() {defaultValue=1, parameterName=Emissive Multiplier})) {materialFunction=&MF_FX_EmissiveAdjustments}
  n90 = Multiply(A=@n89.Result, B=ScalarParameter() {defaultValue=0.4, group=Bubbles, parameterName=Color_Outline})
  n91 = NamedRerouteUsage()
  NamedRerouteDeclaration(None=StaticSwitchParameter(True=LinearInterpolate(A=@n89.Result, B=@n90, Alpha=@n91) {constB=0}, False=@n89.Result) {group=Bubbles, parameterName=outline_bubbles}) {name=Emissive, nodeColor=<rgba 0.6875 1 0 1>}
  n92 = ScalarParameter() {defaultValue=1, group=Bubbles, parameterName=Bubbles_Tile}
  n93 = Multiply(A=@n92) {constA=2}
  n94 = Add(A=NamedRerouteUsage(), B=NamedRerouteUsage())
  n95 = NamedRerouteUsage()
  n96 = Constant2Vector() {g=1, r=-0.04}
  n97 = ScalarParameter() {group=Bubbles, parameterName=Bubbles_Speed}
  n98 = NamedRerouteUsage()
  n99 = Add(A=Panner(Coordinate=Multiply(A=@n94, B=@n93), Time=@n95, Speed=Multiply(A=@n96, B=@n97)) {speedY=3}, B=@n98) {constB=0}
  n100 = TextureSample(UVs=@n99) {
   samplerSource=SSM_Wrap_WorldGroupSettings
   samplerType=SAMPLERTYPE_Normal
   texture=&/T_IronLog_BubbleNormal
  }
  n101 = Multiply(A=@n92) {constA=3, constB=1.333}
  n102 = Constant2Vector() {g=1, r=0.06}
  n103 = Add(A=Panner(Coordinate=Multiply(A=@n94, B=@n101), Time=@n95, Speed=Multiply(A=@n102, B=@n97)) {speedY=3}, B=@n98) {constB=0}
  n104 = TextureSample(UVs=@n103) {
   samplerSource=SSM_Wrap_WorldGroupSettings
   samplerType=SAMPLERTYPE_Normal
   texture=&T_IronLog_BubbleNormal
  }
  n105 = MaterialFunctionCall(Float3=MaterialFunctionCall(BaseNormal=@n100.RGB, AdditionalNormal=@n104.RGB) {materialFunction=&BlendAngleCorrectedNormals}.Result) {materialFunction=&BreakOutFloat3Components}
  n106 = Saturate(None=NamedRerouteUsage())
  NamedRerouteDeclaration(None=Normalize(VectorInput=MaterialFunctionCall(X=@n105.R, Y=@n105.R, Z=@n106) {materialFunction=&MakeFloat3}.Result)) {name=Bubbles Normal, nodeColor=<rgba 0.8047 0 1 1>}
  n107 = NamedRerouteUsage()
  n108 = Multiply(A=NamedRerouteUsage()) {constB=-1}
  n109 = NamedRerouteUsage()
  n110 = LinearInterpolate(
   A=LinearInterpolate(A=NamedRerouteUsage(), B=@n9, Alpha=@n107)
   B=LinearInterpolate(A=@n10, B=Transform(None=@n108) {transformSourceType=TRANSFORMSOURCE_World, transformType=TRANSFORM_Tangent}, Alpha=@n107)
   Alpha=@n109
  )
  n111 = Subtract(A=NamedRerouteUsage()) {constB=0}
  n112 = ScalarParameter() {defaultValue=5, group=Foam, parameterName=Foam Tile}
  n113 = Constant()
  n114 = Multiply(A=NamedRerouteUsage(), B=Constant() {r=1})
  n115 = NamedRerouteDeclaration(None=Add(
    A=Multiply(A=@n111, B=@n112)
    B=MaterialFunctionCall(X=@n113, Y=@n114) {materialFunction=&MakeFloat2}.Result
   ) {constB=0}) {name=FoamUV, nodeColor=<rgba 1 0.75 0 1>}
  n116 = TextureSample(UVs=@n115) {desc=Foam Normal, samplerType=SAMPLERTYPE_Normal, texture=&T_Bubble_B_N}
  n117 = Multiply(A=@n116.B, B=NamedRerouteUsage())
  n118 = Multiply(
   A=Normalize(VectorInput=MaterialFunctionCall(X=@n116.R, Y=@n116.R, Z=@n117) {materialFunction=&MakeFloat3}.Result)
   B=Constant3Vector() {desc=Renormal, constant=<rgba -1 1 1 0>}
  ) {constB=2}
  n119 = LinearInterpolate(A=@n110, B=@n118, Alpha=Multiply(A=@n109, B=NamedRerouteUsage()))
  NamedRerouteDeclaration(None=Normalize(VectorInput=StaticSwitchParameter(True=@n119, False=@n110) {parameterName=Enable Foam})) {name=Pixel Normals, nodeColor=<rgba 0 0 1 1>}
  NamedRerouteDeclaration(None=DotProduct(A=VertexNormalWS(), B=CameraVectorWS())) {name=PixelBumpOffset, nodeColor=<rgba 0.0312 0 1 1>}
  n120 = TextureSample(UVs=@n99) {
   samplerSource=SSM_Wrap_WorldGroupSettings
   samplerType=SAMPLERTYPE_Grayscale
   texture=&/T_IronLog_BubbleMask
  }
  n121 = TextureSample(UVs=@n103) {
   samplerSource=SSM_Wrap_WorldGroupSettings
   samplerType=SAMPLERTYPE_Grayscale
   texture=&T_IronLog_BubbleMask
  }
  n122 = Max(A=@n120.RGB, B=@n121.RGB) {constB=0}
  n123 = ScalarParameter() {group=Bubbles, parameterName=Bubbles Min, sliderMax=1}
  n124 = ScalarParameter() {defaultValue=1, group=Bubbles, parameterName=Bubbles Max, sliderMax=1}
  n125 = NamedRerouteUsage()
  n126 = ScalarParameter() {defaultValue=0.04, group=Bubbles, parameterName=Bubbles Fizziness Float}
  n127 = Add(A=NamedRerouteUsage(), B=Multiply(A=NamedRerouteUsage(), B=@n126) {constB=0.04}) {constB=0.05}
  n128 = Multiply(A=@n126) {constB=0.6}
  n129 = Add(A=NamedRerouteUsage(), B=OneMinus(None=Divide(A=@n127, B=@n128) {constB=0.02})) {constB=0}
  n130 = Multiply(
   A=LinearInterpolate(A=LinearInterpolate(A=@n123, B=@n124, Alpha=@n125), Alpha=Saturate(None=@n129)) {constB=0}
   B=ScalarParameter() {defaultValue=1, group=Bubbles, parameterName=Bubbles Multiplier}
  )
  n131 = SmoothStep(Max=ScalarParameter() {defaultValue=1, parameterName=Bubbles Spherical Mask Clamp}, Value=NamedRerouteUsage()) {constMax=0.5}
  n132 = OneMinus(None=Saturate(None=Multiply(A=@n130, B=@n131)))
  n133 = NamedRerouteDeclaration(None=Divide(A=Subtract(A=@n122, B=@n132), B=OneMinus(None=@n132))) {name=Bubbles Z, nodeColor=<rgba 0 1 0.8125 1>}
  n134 = LinearInterpolate(Alpha=NamedRerouteUsage()) {constA=0.05, constB=2}
  NamedRerouteDeclaration(None=Saturate(None=Divide(A=@n133, B=@n134) {constB=0.5})) {name=Bubbles Mask, nodeColor=<rgba 0.7813 0 1 1>}
  n135 = CameraVectorWS()
  n136 = Normalize(VectorInput=PixelNormalWS())
  n137 = Subtract(A=DotProduct(A=@n135, B=@n136)) {constB=0.15}
  n138 = LinearInterpolate(Alpha=Power(Base=Saturate(None=@n137), Exp=Constant() {r=0.6366}) {constExponent=0.5}) {constA=1, constB=0}
  n139 = SkyLightEnvMapSample(Direction=Subtract(A=Multiply(A=@n138, B=@n136), B=@n135) {constA=0, constB=0}, Roughness=NamedRerouteUsage())
  n140 = NamedRerouteUsage()
  n141 = NamedRerouteUsage()
  n142 = Max(
   A=StaticSwitchParameter(True=LinearInterpolate(A=@n139, B=@n140, Alpha=@n141) {constA=1}, False=@n139) {parameterName=Enable Bubbles}
   B=ScalarParameter() {parameterName=Refraction Min Value, sliderMax=1}
  )
  n143 = OneMinus(None=ScalarParameter() {defaultValue=1, parameterName=Refraction Saturation, sliderMax=1, sliderMin=-1})
  n144 = VectorParameter() {group=Color, parameterName=Color, defaultValue=<rgba 1 1 1 0>}
  n145 = VectorParameter() {group=Color, parameterName=Color Gradient Secondary, defaultValue=<rgba 0.1 0.1 0.1 0>}
  n146 = Divide(A=NamedRerouteUsage(), B=Multiply(A=NamedRerouteUsage()) {constB=0.5})
  n147 = Add(A=Subtract(A=ComponentMask(None=@n146) {b=true}, B=NamedRerouteUsage())) {constB=0}
  n148 = Multiply(
   A=Multiply(A=Multiply(A=@n147) {constB=-0.55}, B=NamedRerouteUsage())
   B=ScalarParameter() {group=Color, parameterName=Color Gradient Multiplier}
  )
  n149 = ScalarParameter() {defaultValue=1, group=Color, parameterName=Color Gradient Power}
  n150 = Saturate(None=Subtract(A=Power(Base=@n148, Exp=@n149) {constExponent=0.75}, B=NamedRerouteUsage()))
  n151 = Multiply(
   A=Add(A=Desaturation(None=@n142, Fraction=@n143), B=NamedRerouteUsage()) {constB=0}
   B=NamedRerouteDeclaration(None=LinearInterpolate(A=@n144.RGB, B=@n145.RGB, Alpha=@n150)) {name=Liquid Color, nodeColor=<rgba 0 0.0625 1 1>}
  )
  n152 = VectorParameter() {group=Foam, parameterName=Foam Color, defaultValue=<rgba 0.4031 0.486 0.6 0>}
  n153 = NamedRerouteUsage()
  n154 = Max(A=Multiply(A=NamedRerouteUsage(), B=NamedRerouteUsage()) {constA=0.5}, B=Multiply(A=Power(Base=@n153)) {constB=0.25})
  n155 = Multiply(
   A=StaticSwitchParameter(True=LinearInterpolate(A=@n151, B=@n152.RGB, Alpha=@n154), False=@n151) {parameterName=Enable Foam}
   B=ScalarParameter() {defaultValue=1, parameterName=Color Multiplier}
  )
  n156 = Multiply(A=@n155, B=ScalarParameter() {defaultValue=0.4, group=Bubbles, parameterName=Color_Outline})
  n157 = NamedRerouteUsage()
  NamedRerouteDeclaration(None=StaticSwitchParameter(True=LinearInterpolate(A=@n155, B=@n156, Alpha=@n157) {constB=0}, False=@n155) {group=Bubbles, parameterName=outline_bubbles}) {name=Color, nodeColor=<rgba 0.5703 1 0 1>}
  n158 = ScalarParameter() {defaultValue=0.025, parameterName=Roughness}
  n159 = ScalarParameter() {defaultValue=0.3, group=Foam, parameterName=Foam Roughness}
  n160 = NamedRerouteUsage()
  NamedRerouteDeclaration(None=StaticSwitchParameter(True=LinearInterpolate(A=@n158, B=@n159, Alpha=@n160) {desc=roughness}, False=@n158) {parameterName=Enable Foam}) {name=Roughness, nodeColor=<rgba 1 0 0.4453 1>}
  n161 = NamedRerouteUsage()
  n162 = ScalarParameter() {defaultValue=1, group=UVs, parameterName=Water Turbulence Tile}
  n163 = NamedRerouteUsage()
  n164 = ScalarParameter() {defaultValue=1, group=UVs, parameterName=Water Turbulence Speed}
  n165 = Sine(None=Panner(Coordinate=Multiply(A=@n161, B=@n162), Time=Multiply(A=@n163, B=@n164)) {speedX=1, speedY=1})
  n166 = Max(A=NamedRerouteUsage(), B=ScalarParameter() {group=UVs, parameterName=Water Turbulence Min})
  n167 = Multiply(
   A=Multiply(A=ScalarParameter() {defaultValue=1, group=UVs, parameterName=Water Turbulence Intensity}, B=Constant2Vector() {g=0.2, r=1})
   B=LinearInterpolate(Alpha=NamedRerouteUsage()) {constA=1, constB=0.25}
  )
  NamedRerouteDeclaration(None=Multiply(A=Multiply(A=@n165) {constB=0.03}, B=Multiply(A=@n166, B=@n167))) {name=UVs - Water Turbulence, nodeColor=<rgba 1 0 0.8203 1>}
  n168 = ScalarParameter() {defaultValue=0.95, parameterName=Specular}
  n169 = ScalarParameter() {defaultValue=0.5, group=Foam, parameterName=Foam Specular}
  n170 = NamedRerouteUsage()
  NamedRerouteDeclaration(None=StaticSwitchParameter(True=LinearInterpolate(A=@n168, B=@n169, Alpha=@n170) {desc=roughness}, False=@n168) {parameterName=Enable Foam}) {name=Specular, nodeColor=<rgba 0 0.9297 1 1>}
  n171 = Multiply(A=NamedRerouteUsage(), B=ScalarParameter() {defaultValue=5, group=Foam, parameterName=Foam Tile}) {constB=2}
  n172 = NamedRerouteUsage()
  NamedRerouteDeclaration(None=TextureSampleParameter2D(UVs=LinearInterpolate(A=@n171, B=@n115, Alpha=@n172)) {parameterName=Mask_Top_Foam, samplerType=SAMPLERTYPE_LinearGrayscale, texture=/Game/Global/Textures/Default/Noise/T_Bubble_B}.R) {name=Foam Height, nodeColor=<rgba 0 0.9297 1 1>}
  n173 = NamedRerouteUsage()
  n174 = Constant()
  n175 = Constant() {r=0.6}
  n176 = MaterialFunctionCall(X=@n174, Y=@n175) {materialFunction=&MakeFloat2}
  n177 = TextureSampleParameter2D(UVs=Add(A=Panner(Coordinate=@n27, Time=@n173, Speed=@n176.Result) {speedY=0.75}) {constB=0}) {parameterName=Normal_Top, samplerType=SAMPLERTYPE_Normal, texture=&T_Bubble_B_N}
  n178 = MaterialFunctionCall!(Angle=Constant() {r=0.25}, UV/2DVector=@n27) {materialFunction=/Game/Characters/Player/Male/Medium/Bodies/M_MED_Faux/Materials/MF_2DRotate}
  n179 = MaterialFunctionCall(X=@n175, Y=@n174) {materialFunction=&MakeFloat2}
  n180 = TextureSampleParameter2D(UVs=Add(A=Panner(Coordinate=@n178.Result, Time=@n173, Speed=@n179.Result) {speedX=0.75}) {constB=0}) {parameterName=Normal_Top, samplerType=SAMPLERTYPE_Normal, texture=&T_Bubble_B_N}
  n181 = Multiply(
   A=MaterialFunctionCall(X=@n180.G, Y=@n180.G, Z=@n180.G) {materialFunction=&MakeFloat3}.Result
   B=Constant3Vector() {constant=<rgba 1 -1 1 0>}
  )
  n182 = Multiply(
   A=Multiply(
    A=MaterialFunctionCall(BaseNormal=@n177.RGB, AdditionalNormal=@n181) {materialFunction=&BlendAngleCorrectedNormals}.Result
    B=Constant3Vector() {desc=Renormal, constant=<rgba 1 -1 -1 0>}
   ) {constB=-1}
   B=LinearInterpolate(A=Constant3Vector() {constant=<rgba 0.01 0.01 1 0>}, Alpha=NamedRerouteUsage()) {constA=1, constAlpha=1}
  )
  n183 = TextureSampleParameter2D(UVs=Multiply(A=NamedRerouteUsage(), B=ScalarParameter() {defaultValue=5, group=Foam, parameterName=Foam Tile}) {constB=2}) {parameterName=Normal_Top_Foam, samplerType=SAMPLERTYPE_Normal, texture=&T_Bubble_B_N}
  n184 = Multiply(A=@n183.B, B=NamedRerouteUsage())
  n185 = Multiply(
   A=Normalize(VectorInput=MaterialFunctionCall(X=@n183.R, Y=@n183.R, Z=@n184) {materialFunction=&MakeFloat3}.Result)
   B=Constant3Vector() {desc=Renormal, constant=<rgba -1 1 -1 0>}
  )
  n186 = NamedRerouteUsage()
  n187 = MaterialFunctionCall(
   BaseNormal=NamedRerouteUsage()
   AdditionalNormal=StaticSwitchParameter(True=LinearInterpolate(A=@n182, B=@n185, Alpha=@n186), False=@n182) {parameterName=Enable Foam}
  ) {materialFunction=&BlendAngleCorrectedNormals}
  n188 = NamedRerouteUsage()
  NamedRerouteDeclaration(None=Transform(None=StaticSwitchParameter(True=@n187.Result, False=@n188) {defaultValue=true, parameterName=Is Niagara}) {transformSourceType=TRANSFORMSOURCE_World, transformType=TRANSFORM_Tangent}) {name=Water Surface Normal, nodeColor=<rgba 0 0.6016 1 1>}
  NamedRerouteDeclaration(None=@n70) {name=WorldCoordinates, nodeColor=<rgba 1 0 0.3984 1>}
  NamedRerouteDeclaration(None=@n19) {name=Density, nodeColor=<rgba 1 0 0.3281 1>}
  n189 = DynamicParameter() {defaultValue=<rgba 0 0 1 0.5>, paramNames [Water Normal X, Water Normal Y, Water Normal Z, Fill]}
  NamedRerouteDeclaration(None=Normalize(VectorInput=MaterialFunctionCall(X=@n189.Water Normal X, Y=@n189.Water Normal X, Z=@n189.Water Normal X) {materialFunction=&MakeFloat3}.Result)) {name=WaterPlaneZ, nodeColor=<rgba 1 0 0.1875 1>}
  n190 = NamedRerouteUsage()
  NamedRerouteDeclaration(None=Add(A=NamedRerouteUsage(), B=LinearInterpolate(Alpha=@n190) {constA=0.25, constB=0})) {name=Bubbles UV, nodeColor=<rgba 0.1953 0 1 1>}
  n191 = NamedRerouteUsage()
  n192 = Time()
  n193 = NamedRerouteUsage()
  NamedRerouteDeclaration(None=Multiply(
    A=StaticSwitchParameter(True=@n191, False=@n192) {defaultValue=true, parameterName=Is Niagara}
    B=LinearInterpolate(Alpha=@n193) {constA=1, constB=0.2}
   )) {name=Viscosity Time, nodeColor=<rgba 0 1 0.5078 1>}
  n194 = WorldPosition() {worldPositionShaderOffset=WPT_ExcludeAllShaderOffsets}
  n195 = CameraPositionWS()
  n196 = DotProduct(A=VertexNormalWS(), B=Normalize(VectorInput=Subtract(A=@n194.XYZ, B=@n195)))
  n197 = Multiply(B=Multiply(A=ScalarParameter() {defaultValue=1, parameterName=WPO amount}, B=NamedRerouteUsage()) {constA=1}) {constA=0.5}
  n198 = Multiply(
   A=SmoothStep(Max=Constant() {r=0.001}, Value=@n196)
   B=MaterialFunctionCall(Offset Amount=Multiply(A=@n196, B=@n197)) {materialFunction=/Engine/Functions/Engine_MaterialFunctions02/WorldPositionOffset/CameraOffset}.Clamped Camera Offset
  )
  n199 = Constant3Vector()
  NamedRerouteDeclaration(None=Add(A=ShadowReplace(Default=@n198, Shadow=@n199), B=NamedRerouteUsage())) {name=WPO, nodeColor=<rgba 0 0.5 1 1>}
  n200 = ComponentMask(None=@n24) {g=true, r=true}
  n201 = ComponentMask(None=@n23) {b=true}
  NamedRerouteDeclaration(None=AppendVector(
    A=MaterialFunctionCall(Vector or UVs=@n200) {materialFunction=&VectorToRadialValue}.Vector Converted to Angle
    B=Multiply(A=@n201) {constB=-1}
   )) {name=Spherical UV, nodeColor=<rgba 1 0.1641 0 1>}
  n202 = CameraVectorWS()
  n203 = Normalize(VectorInput=NamedRerouteUsage())
  n204 = DotProduct(A=@n202, B=@n203)
  n205 = Power(Base=Saturate(None=Subtract(A=@n204) {constB=0.15}), Exp=Constant() {r=0.6366}) {constExponent=0.5}
  n206 = Subtract(A=Multiply(A=LinearInterpolate(Alpha=@n205) {constA=1, constB=0}, B=@n203), B=@n202) {constA=0, constB=0}
  n207 = NamedRerouteUsage()
  n208 = DotProduct(A=NamedRerouteUsage(), B=AtmosphericLightVector())
  n209 = Multiply(
   A=LinearInterpolate(Alpha=ConstantBiasScale(None=@n208)) {constA=0.1}
   B=ScalarParameter() {defaultValue=1, group=Bubbles, parameterName=Bubbles Brightness}
  ) {constB=0.05}
  n210 = Constant() {r=0.6}
  n211 = MaterialFunctionCall(Color_Morning=@n210, Color_Day=Constant() {r=1}, Color_Evening=@n210, Color_Night=Constant() {r=0.2}) {materialFunction=&Color_Time_of_Day}
  NamedRerouteDeclaration(None=Add(A=SkyLightEnvMapSample(Direction=@n206, Roughness=@n207), B=Multiply(A=@n209, B=@n211.Result))) {name=Bubbles Color, nodeColor=<rgba 0 0.0156 1 1>}
  NamedRerouteDeclaration() {name=TestingGrad, nodeColor=<rgba 0 0.25 1 1>}
  NamedRerouteDeclaration(None=@n84) {name=Top Mask, nodeColor=<rgba 0 0.8594 1 1>}
  n212 = NamedRerouteUsage()
  NamedRerouteDeclaration(None=Multiply(A=ComponentMask(None=@n212) {g=true, r=true}, B=Constant2Vector() {g=-0.5, r=0.5})) {name=Water Sway, nodeColor=<rgba 1 0 0.7734 1>}
  n213 = OneMinus(None=Abs(None=@n25))
  n214 = Constant() {r=0.0025}
  NamedRerouteDeclaration(None=Divide(A=Subtract(A=@n213, B=@n214) {constB=0.1}, B=OneMinus(None=@n214)) {constB=0.9}) {name=Spherical UV Mask, nodeColor=<rgba 1 0 0.4453 1>}
  n215 = Multiply(A=NamedRerouteUsage())
  n216 = ComponentMask(None=NamedRerouteUsage()) {b=true}
  n217 = Saturate(None=Multiply(A=Multiply(A=@n216, B=@n76) {constB=10}) {constB=0.7})
  n218 = Divide(A=Abs(None=@n74), B=@n76)
  n219 = OneMinus(None=Power(Base=Saturate(None=@n218)) {constExponent=3})
  n220 = LinearInterpolate(Alpha=LinearInterpolate(A=Power(Base=@n217) {constExponent=0.7}, B=Multiply(A=@n217, B=@n219), Alpha=NamedRerouteUsage())) {constA=-1, constB=0.15}
  n221 = Multiply(A=SmoothStep(Value=Add(A=@n215, B=@n220)) {constMax=0.15}, B=ScalarParameter() {defaultValue=1, group=Foam, parameterName=Foam Intensity})
  NamedRerouteDeclaration(None=StaticSwitchParameter(True=Saturate(None=@n221), False=Constant()) {parameterName=Enable Foam}) {name=FoamTop, nodeColor=<rgba 0.4297 0 1 1>}
  NamedRerouteDeclaration(None=LinearInterpolate(A=NamedRerouteUsage(), Alpha=NamedRerouteUsage()) {constB=0}) {name=Base Color, nodeColor=<rgba 0.3594 1 0 1>}
  NamedRerouteDeclaration(None=@n73.Transformed Vector) {name=Rotated World Coord, nodeColor=<rgba 1 0.2109 0 1>}
  n222 = NamedRerouteUsage()
  n223 = NamedRerouteUsage()
  n224 = MaterialFunctionCall(
   Plane Normal Axis=NamedRerouteUsage()
   Plane X Axis=NamedRerouteUsage()
   Plane Center=Subtract(A=Add(A=@n222, B=@n223), B=NamedRerouteUsage())
   UV Scale=Multiply(A=NamedRerouteUsage(), B=ScalarParameter() {defaultValue=2, parameterName=Virtual Plane UV Scale})
  ) {materialFunction=&VirtualPlaneCoordinates}
  NamedRerouteDeclaration(None=@n224.World Position - Centered) {name=VP - World Position Local, nodeColor=<rgba 0.875 0 1 1>}
  NamedRerouteDeclaration(None=@n224.World Position) {name=VP - World Position, nodeColor=<rgba 0 0.4844 1 1>}
  NamedRerouteDeclaration(None=@n224.UVs) {name=VP - UVs, nodeColor=<rgba 0 0.1562 1 1>}
  NamedRerouteDeclaration(None=@n31.Time) {name=Niagara Time, nodeColor=<rgba 0 1 0.6719 1>}
  NamedRerouteDeclaration(None=ScalarParameter() {parameterName=Density}) {name=Liquid Density, nodeColor=<rgba 1 0.1406 0 1>}
  NamedRerouteDeclaration(None=ScalarParameter() {defaultValue=0.5, group=Foam, parameterName=Foam Density}) {name=Foam Density, nodeColor=<rgba 1 0.5391 0 1>}
  n225 = PixelNormalWS()
  n226 = DotProduct(A=CameraVectorWS(), B=@n225)
  n227 = Power(Base=Multiply(A=OneMinus(None=@n226), B=NamedRerouteUsage()), Exp=ScalarParameter() {defaultValue=1, parameterName=Spherical UV Power}) {constExponent=6}
  n228 = ScalarParameter() {defaultValue=1, group=UVs, parameterName=Spherical UV Push}
  n229 = ComponentMask(None=Transform(None=@n225) {transformSourceType=TRANSFORMSOURCE_World, transformType=TRANSFORM_Camera}) {g=true, r=true}
  n230 = Constant2Vector() {g=-1, r=1}
  NamedRerouteDeclaration(None=Multiply(A=Multiply(A=@n227, B=@n228), B=Multiply(A=@n229, B=@n230))) {name=UVs - Vertex Normal Distortion, nodeColor=<rgba 0.6172 0 1 1>}
  NamedRerouteDeclaration(None=ScalarParameter() {parameterName=Viscosity}) {name=Viscosity, nodeColor=<rgba 0.1484 1 0 1>}
  NamedRerouteDeclaration(None=ScalarParameter() {group=Bubbles, parameterName=Bubbles Fizziness}) {name=Bubbles - Fizziness, nodeColor=<rgba 0 1 0.4375 1>}
  n231 = MaterialFunctionCall() {materialFunction=&ObjectScale}
  NamedRerouteDeclaration(None=Max(A=Max(A=@n231.Scale X, B=@n231.Scale X), B=@n231.Scale Z)) {name=Liquid Scale, nodeColor=<rgba 0.8047 0 1 1>}
  n232 = VectorParameter() {parameterName=WorldPostionOfBlackHole, defaultValue=<rgba -30000 25536 1952 0>}
  n233 = AppendVector(A=@n232.RGB, B=@n232.RGB)
  n234 = WorldPosition()
  NamedRerouteDeclaration(None=StaticSwitchParameter(
    True=Constant()
    False=MaterialFunctionCall!(WorldPostionOfBlackHole=@n233, World Position=@n234.XYZ) {materialFunction=/Game/Packages/Fortnite_UniqueMaterials_1/BlackHoleFunction}.Result
   ) {defaultValue=true, parameterName=Is Niagara}) {name=Backpack WPO, nodeColor=<rgba 1 0 0.1406 1>}
  n235 = PreSkinnedLocalBounds()
  n236 = Multiply(A=Add(A=@n235.Min, B=@n235.Min)) {constB=0.5}
  n237 = ComponentMask(None=VectorParameter() {group=Static Overrides, parameterName=Center Offset}.RGB) {b=true, g=true, r=true}
  NamedRerouteDeclaration(None=StaticSwitchParameter(
    True=Transform(None=@n236) {transformSourceType=TRANSFORMSOURCE_Instance}
    False=Transform(None=@n237) {transformSourceType=TRANSFORMSOURCE_Local}
   ) {defaultValue=true, parameterName=Is Niagara}) {name=Center Offset, nodeColor=<rgba 1 0 0.7266 1>}
  NamedRerouteDeclaration(None=ScalarParameter() {parameterName=Metallic}) {name=Metallic, nodeColor=<rgba 1 0.5156 0 1>}
  n238 = ScalarParameter() {parameterName=Liquid Fresnel Glow}
  n239 = Max(
   A=Fresnel(ExponentIn=ScalarParameter() {defaultValue=3, parameterName=Liquid Fresnel Glow Exp}, Normal=VertexNormalWS()) {baseReflectFraction=0, exponent=3}
   B=Multiply(A=NamedRerouteUsage()) {constB=0.5}
  )
  NamedRerouteDeclaration(None=Multiply(A=NamedRerouteUsage(), B=Multiply(A=@n238, B=@n239))) {name=Fresnel Glow, nodeColor=<rgba 1 0.6328 0 1>}
  n240 = Multiply(A=NamedRerouteUsage()) {constB=2}
  n241 = Multiply(A=OneMinus(None=Saturate(None=@n240)), B=OneMinus(None=NamedRerouteUsage()))
  n242 = NamedRerouteUsage()
  NamedRerouteDeclaration(None=LinearInterpolate(A=@n9, B=NamedRerouteUsage(), Alpha=Multiply(A=@n241, B=@n242))) {name=Bubbles Normal Masked, nodeColor=<rgba 0.3594 1 0 1>}
  n243 = LinearInterpolate(Alpha=NamedRerouteUsage()) {constB=0.1}
  n244 = ScalarParameter() {parameterName=Refraction Roughness, sliderMax=1}
  n245 = NamedRerouteUsage()
  NamedRerouteDeclaration(None=Add(A=Add(A=@n243, B=@n244), B=LinearInterpolate(Alpha=@n245) {constA=1, constB=0})) {name=Refraction Roughness, nodeColor=<rgba 0 0.4141 1 1>}
  NamedRerouteDeclaration(None=@n30.Turbulence) {name=Turbulence, nodeColor=<rgba 0 0.7891 1 1>}
  n246 = NamedRerouteUsage()
  n247 = NamedRerouteUsage()
  n248 = Power(
   Base=OneMinus(None=ComponentMask(None=@n247) {b=true})
   Exp=ScalarParameter() {defaultValue=7.9818, group=Bubbles, parameterName=Outline_Pow_base}
  )
  n249 = NamedRerouteUsage()
  n250 = Multiply(
   A=Multiply(A=Multiply(A=@n246, B=@n248), B=LinearInterpolate(Alpha=@n249) {constA=0.25})
   B=ScalarParameter() {defaultValue=25.1755, group=Bubbles, parameterName=outline_Mult}
  )
  n251 = ScalarParameter() {defaultValue=0.9572, group=Bubbles, parameterName=Outline_Pow}
  NamedRerouteDeclaration(None=Saturate(None=Power(Base=@n250, Exp=@n251))) {name=bubbles outline, nodeColor=<rgba 1 0.0703 0 1>}
  NamedRerouteDeclaration(None=@n30.Param3) {name=Foam Time, nodeColor=<rgba 0 1 0.7656 1>}
  n252 = NamedRerouteUsage()
  n253 = Constant3Vector() {constant=<rgba 1 -1 -1 0>}
  NamedRerouteDeclaration(None=Transform(None=Multiply(A=@n252, B=@n253)) {transformSourceType=TRANSFORMSOURCE_View}) {name=Bubbles Normal World Space, nodeColor=<rgba 1 0 0.6797 1>}
  n254 = NamedRerouteUsage()
  n255 = ComponentMask(None=NamedRerouteUsage()) {b=true}
  n256 = Multiply(A=@n254, B=Power(Base=OneMinus(None=@n255)))
  n257 = NamedRerouteUsage()
  NamedRerouteDeclaration(None=Multiply(A=StaticSwitchParameter(True=@n256, False=@n254) {parameterName=CullInterior}, B=LinearInterpolate(Alpha=@n257) {constA=0.25})) {name=Bubbles Color Mask, nodeColor=<rgba 0.0312 0 1 1>}
  NamedRerouteDeclaration(None=@n118) {name=Foam Normal, nodeColor=<rgba 0.2188 0 1 1>}
  n258 = Time()
  n259 = ScalarParameter() {defaultValue=0.1, parameterName=Waterlevel Breakup Rotation Speed, sliderMax=1}
  n260 = MaterialFunctionCall(Float2=RotateAboutAxis(
    NormalizedRotationAxis=Constant3Vector() {constant=<rgba 0 0 1 0>}
    RotationAngle=Multiply(A=@n258, B=@n259)
    PivotPoint=Constant3Vector()
    Position=LocalPosition() {includedOffsets=ExcludeOffsets}.XYZ
   )) {materialFunction=&BreakOutFloat2Components}
  n261 = Multiply(A=Arctangent2!(Y=@n260.R, X=@n260.R), B=ScalarParameter() {defaultValue=2, parameterName=Waterlevel Breakup Tiling, sliderMax=2}) {constB=2}
  n262 = Divide(A=ParticleSpeed(), B=ScalarParameter() {defaultValue=250, parameterName=Waterlevel Breakup Vel Threshold, sliderMax=500}) {constB=250}
  n263 = Multiply(
   A=Power(
    Base=Sine(None=@n261) {period=3.14}
    Exp=ScalarParameter() {defaultValue=1, parameterName=Waterlevel Breakup Profile Exponent, sliderMax=2, sliderMin=0.032}
   )
   B=Multiply(A=Saturate(None=@n262), B=ScalarParameter() {parameterName=Waterlevel Breakup, sliderMax=1})
  )
  NamedRerouteDeclaration(None=StaticSwitchParameter(True=Add(A=@n189.Fill, B=@n263), False=ScalarParameter() {defaultValue=1, group=Static Overrides, parameterName=Liquid Fill}) {defaultValue=true, parameterName=Is Niagara}) {name=Fill, nodeColor=<rgba 0 1 0.0859 1>}
  n264 = Multiply(A=@n235.Extents, B=@n231.Scale XYZ)
  NamedRerouteDeclaration(None=@n264) {name=Dimensions, nodeColor=<rgba 0 0.0994 1 1>}
  n265 = NamedRerouteUsage()
  NamedRerouteDeclaration(None=Multiply(A=NamedRerouteUsage(), B=Multiply(A=@n265, B=@n37))) {name=Center, nodeColor=<rgba 1 0 0.0937 1>}
  NamedRerouteUsage()
  NamedRerouteUsage()
  NamedRerouteUsage()
  NamedRerouteUsage()
  NamedRerouteUsage()
  NamedRerouteUsage()
  NamedRerouteUsage()
  NamedRerouteUsage()
  NamedRerouteUsage()
  NamedRerouteUsage()
  NamedRerouteUsage()
  NamedRerouteUsage()
  NamedRerouteUsage()
  NamedRerouteUsage()
  NamedRerouteUsage()
  NamedRerouteUsage()
  OneMinus(None=Multiply(A=NamedRerouteUsage(), B=NamedRerouteUsage()) {constA=0.5})
  Power(Base=@n25, Exp=ScalarParameter() {defaultValue=1, parameterName=Spherical UV Power}) {constExponent=3}
  n266 = MaterialFunctionCall(Float2=LocalPosition() {includedOffsets=ExcludeOffsets}.XY) {materialFunction=&BreakOutFloat2Components}
  Sine(None=Multiply(A=Arctangent2(Y=@n266.R, X=@n266.R)) {constB=2}) {period=3.14}
  StaticSwitchParameter(True=@n264, False=VectorParameter() {group=Static Overrides, parameterName=Container Size, defaultValue=<rgba 1 1 1 1>}.RGB) {defaultValue=true, parameterName=Is Niagara}
  n267 = NamedRerouteUsage()
  n268 = NamedRerouteUsage()
  n269 = NamedRerouteUsage()
  StaticSwitchParameter(
   True=LinearInterpolate(
    A=@n267
    B=VectorParameter() {group=Foam, parameterName=Foam Color, defaultValue=<rgba 0.4031 0.486 0.6 0>}.RGB
    Alpha=Multiply(A=@n268, B=@n269) {constA=0.5}
   )
   False=@n267
  ) {parameterName=Enable Foam}
  TextureCoordinate()
  TextureSample() {samplerType=SAMPLERTYPE_Grayscale, texture=/Game/Effects/Niagara/MasterMaterials/Liquid/T_Liquid_Bubbles_Mask}
  TextureSample() {samplerType=SAMPLERTYPE_Normal, texture=/Game/Effects/Niagara/MasterMaterials/Liquid/T_Liquid_Bubbles_Normal}
  Transform() {transformSourceType=TRANSFORMSOURCE_World, transformType=TRANSFORM_Tangent}
  VectorParameter() {group=Potion, parameterName=Water Turbulence Amount, defaultValue=<rgba 1 0 1 1>}
  VertexNormalWS()
 ]
}
